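# M_Center_Glow is an unlit, two-sided additive material for a glow in the S7 environment effects, enabled for sprite and mesh particles, including Niagara.
# The green channel of a flare texture parameter, defaulting to T_Flare_Spikeball, is multiplied by particle alpha.
# A DepthFade with a long default distance takes that product as opacity to soften intersections.
# The graph shows no connection from this chain to a material output.
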Material /SparksMaps/Effects/Environment/S7/Materials/M_Center_Glow
graph {
 properties {
  bAllowNegativeEmissiveColor=true
  bUsedWithMeshParticles=true
  bUsedWithNanite=true
  bUsedWithNiagaraMeshParticles=true
  bUsedWithNiagaraSprites=true
  bUsedWithParticleSprites=true
  blendMode=BLEND_Additive
  pixelDepthOffsetMode=PDOM_Legacy
  shadingModel=MSM_Unlit
  twoSided=true
  usedShadingModels=Unlit
 }
 nodes [DepthFade(Opacity=Multiply(
    A=TextureSampleParameter2D() {parameterName=Param, texture=/SparksMaps/Effects/Environment/S7/Textures/T_Flare_Spikeball}.G
    B=ParticleColor().A
   )) {fadeDistanceDefault=3000}]
}
