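# M_Stage_Glow is an unlit additive material for a pulsing, cyan-by-default environment glow, apparently around a stage.
# A UV mask multiplies a vertical gradient by a horizontal falloff that fades at both side edges, sharpened by the Vertical_Pow parameter and softened by a long-distance DepthFade.
# Two offset sine waves at different speeds multiply and clamp into an irregular brightness pulse that scales the HDR color parameter.
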Material /SparksMaps/Effects/Environment/S7/Materials/M_Stage_Glow
graph {
 properties {bAutomaticallySetUsageInEditor=false, blendMode=BLEND_Additive, shadingModel=MSM_Unlit, usedShadingModels=Unlit}
 nodes [
  n1 = TextureCoordinate()
  n2 = ComponentMask(None=@n1) {g=true}
  n3 = ComponentMask(None=@n1) {r=true}
  n4 = Multiply(A=@n3, B=OneMinus(None=@n3))
  DepthFade(Opacity=Power(Base=Multiply(A=@n2, B=@n4), Exp=ScalarParameter() {defaultValue=2, parameterName=Vertical_Pow})) {fadeDistanceDefault=5000}
  n5 = Time()
  n6 = Add(A=Sine(None=Multiply(A=@n5) {constB=1.2}))
  n7 = Time()
  n8 = Add(A=Sine(None=Multiply(A=@n7) {constB=0.65}))
  Multiply(A=Clamp(None=Multiply(A=@n6, B=@n8)) {maxDefault=2, minDefault=1}, B=VectorParameter() {defaultValue=<rgba 0 50 50 0>}.RGB)
 ]
}
