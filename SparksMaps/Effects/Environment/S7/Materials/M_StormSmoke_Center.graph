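# Unlit, two-sided translucent storm smoke material for Niagara sprite and mesh particles, by its name the center of a storm.
# Two panning samples of a tiling cloud texture, scaled through dynamic parameters, multiply to blend two tint colors, then multiply by particle color.
# Opacity comes from particle alpha times an inverted, contrast-adjusted storm texture green channel, softened by a long depth fade.
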Material /SparksMaps/Effects/Environment/S7/Materials/M_StormSmoke_Center
graph {
 properties {
  bAutomaticallySetUsageInEditor=false
  bUsedWithNiagaraMeshParticles=true
  bUsedWithNiagaraSprites=true
  blendMode=BLEND_Translucent
  shadingModel=MSM_Unlit
  twoSided=true
  usedShadingModels=Unlit
 }
 nodes [
  n1 = ParticleColor()
  n2 = Power(
   Base=TextureSample() {texture=/SparksMaps/Effects/Environment/S7/Textures/Storm_Tex}.G
   Exp=ScalarParameter() {defaultValue=0.8, parameterName=Cloud_Contrast}
  )
  DepthFade(Opacity=Multiply(A=@n1.A, B=OneMinus(None=@n2))) {fadeDistanceDefault=5000}
  n3 = DynamicParameter() {parameterIndex=1}
  n4 = Multiply(A=TextureCoordinate(), B=AppendVector(A=@n3.Param1, B=@n3.Param1))
  n5 = Time()
  n6 = TextureSample(UVs=Panner(Coordinate=Multiply(A=@n4, B=@n3.Param3), Time=@n5, Speed=Constant2Vector() {g=-0.004, r=0.004}) {speedX=-0.009, speedY=0.008}) {texture=/SparksEffects/Effects/Environment/S7/Textures/T_TilingClouds_D}
  n7 = TextureSample(UVs=Panner(Coordinate=@n4, Time=@n5, Speed=Constant2Vector() {g=0.007, r=0.008}) {speedX=0.01, speedY=0.009}) {texture=&T_TilingClouds_D}
  Multiply(
   A=LinearInterpolate(
    A=VectorParameter() {parameterName=cc, defaultValue=<rgba 0.0894 0.0651 0.25 1>}.RGB
    B=VectorParameter() {parameterName=cc1, defaultValue=<rgba 0.6638 0.3779 0.8333 1>}.RGB
    Alpha=Multiply(A=@n6.RGB, B=@n7.RGB)
   )
   B=@n1.RGB
  )
 ]
}
