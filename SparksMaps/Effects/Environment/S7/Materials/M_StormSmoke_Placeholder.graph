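# Translucent two-sided placeholder storm smoke material for Niagara sprite and mesh particles, previewed on a storm cloud mesh.
# Two panned tiling cloud samples multiply together, tinted by a blend of two dynamic-parameter colors, whose fourth channels also drive tiling and pan speed.
# That result feeds both a particle-color multiply and a center-to-outer color gradient masked by storm texture red.
# Storm texture green, scaled by AlphaMult, feeds a long depth fade.
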Material /SparksMaps/Effects/Environment/S7/Materials/M_StormSmoke_Placeholder
graph {
 properties {
  bAutomaticallySetUsageInEditor=false
  bUsedWithNiagaraMeshParticles=true
  bUsedWithNiagaraSprites=true
  blendMode=BLEND_Translucent
  previewMesh=/SparksMaps/Effects/Environment/S7/Meshes/SM_Storm_Cloud
  twoSided=true
  usedShadingModels=Default Lit
 }
 nodes [
  n1 = TextureSample() {texture=/SparksMaps/Effects/Environment/S7/Textures/Storm_Tex}
  DepthFade(Opacity=Multiply(A=@n1.G, B=ScalarParameter() {defaultValue=1.5, parameterName=AlphaMult})) {fadeDistanceDefault=5000}
  n2 = TextureCoordinate()
  n3 = DynamicParameter() {parameterIndex=1, paramNames [R,G,B,A]}
  n4 = Time()
  n5 = DynamicParameter() {parameterIndex=2}
  n6 = Panner(Coordinate=Multiply(A=@n2, B=@n3.A), Time=@n4, Speed=Multiply(A=@n5.Param4, B=Constant2Vector() {g=-0.004, r=0.004})) {speedX=-0.009, speedY=0.008}
  n7 = Constant2Vector() {g=0.007, r=0.008}
  n8 = TextureSample(UVs=Panner(Coordinate=Multiply(A=@n2, B=@n5.Param4), Time=@n4, Speed=Multiply(A=@n3.A, B=@n7)) {speedX=0.01, speedY=0.009}) {texture=/SparksEffects/Effects/Environment/S7/Textures/T_TilingClouds_D}
  n9 = Multiply(
   A=Multiply(A=TextureSample(UVs=@n6) {texture=&T_TilingClouds_D}.RGB, B=@n8.RGB)
   B=LinearInterpolate(A=@n3.RGB, B=@n5.RGB, Alpha=@n8.A)
  )
  n10 = ParticleColor()
  Multiply(A=Multiply(A=@n9, B=@n10.RGB), B=@n10.A)
  Multiply(
   A=LinearInterpolate(
    A=VectorParameter() {parameterName=color_center, defaultValue=<rgba 0 0 0 1>}.RGB
    B=VectorParameter() {parameterName=color_out, defaultValue=<rgba 2 2 2 1>}.RGB
    Alpha=@n1.R
   )
   B=@n9
  )
 ]
}
