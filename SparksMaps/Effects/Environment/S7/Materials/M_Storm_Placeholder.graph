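# Placeholder storm material: an unlit, translucent, two-sided Niagara sprite effect that draws a portal-like swirl.
# Radial, tilted and panning UV chains are shared through named reroutes, and a tech tile texture is sampled at a mip level set by Softness.
# That shaded sample modulates radial sphere masks, shaping both the depth-faded opacity and the portal gradient.
# The gradient tints light over shadow color, multiplied by particle color for emissive.
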Material /SparksMaps/Effects/Environment/S7/Materials/M_Storm_Placeholder
graph {
 properties {
  bAutomaticallySetUsageInEditor=false
  bUsedWithNiagaraSprites=true
  blendMode=BLEND_Translucent
  shadingModel=MSM_Unlit
  twoSided=true
  usedShadingModels=Unlit
 }
 nodes [
  n1 = TextureCoordinate()
  BumpOffset(Coordinate=@n1, Height=ScalarParameter() {defaultValue=-12, group=UV, parameterName=ParallaxIntensity}) {referencePlane=0}
  n2 = ScalarParameter() {defaultValue=0.081, group=Opacity, parameterName=PortalSize}
  n3 = Constant() {r=0.5}
  n4 = SphereMask(A=TextureCoordinate(), B=@n3, Radius=@n3) {hardnessPercent=0}
  n5 = TextureSampleParameter2D!(UVs=NamedRerouteUsage(), Level=ScalarParameter() {defaultValue=1.5, group=Shape, parameterName=Softness}, Apply View MipBias) {
   mipValueMode=TMVM_MipLevel
   parameterName=MainTex
   samplerType=SAMPLERTYPE_LinearColor
   texture=/SparksMaps/Effects/Environment/S7/Textures/Tile_Tech_02
  }
  n6 = LinearInterpolate(
   A=ScalarParameter() {defaultValue=0.6132, group=Shape, parameterName=Shade}
   B=ScalarParameter() {defaultValue=-0.2691, group=Shape, parameterName=Highlight}
   Alpha=SquareRoot(None=@n5.G)
  )
  n7 = Multiply(A=@n4, B=OneMinus(None=@n6))
  n8 = Saturate(None=Power(Base=InverseLinearInterpolate(B=@n2, Value=@n7) {constB=0.2}) {constExponent=4})
  n9 = ScalarParameter() {defaultValue=1, group=Opacity, parameterName=MasterOpacity}
  n10 = OneMinus(None=@n4)
  DepthFade(Opacity=Multiply(A=Multiply(A=@n8, B=@n9), B=Power(Base=@n10))) {fadeDistanceDefault=3000}
  n11 = TextureCoordinate()
  n12 = Constant() {r=0.5}
  n13 = ScalarParameter() {defaultValue=0.5, parameterName=GradientStart}
  n14 = NamedRerouteUsage()
  n15 = Constant() {r=0.5}
  n16 = ScalarParameter() {defaultValue=1.1629, parameterName=Gradient_Radius}
  n17 = Max(
   A=OneMinus(None=SphereMask(A=@n11, B=@n12, Radius=@n13, Hardness=@n6) {hardnessPercent=50})
   B=OneMinus(None=SphereMask(A=@n14, B=@n15, Radius=@n16, Hardness=@n6) {hardnessPercent=50})
  )
  n18 = OneMinus(None=SphereMask(A=NamedRerouteUsage(), B=Constant() {r=0.5}, Radius=Constant() {r=0.676}, Hardness=Constant() {r=0.2452}) {hardnessPercent=50})
  n19 = Multiply(A=Power(Base=Multiply(A=@n17, B=@n18), Exp=Constant() {r=2}), B=ScalarParameter() {defaultValue=2, parameterName=PortalIntensity})
  n20 = Multiply(A=@n19, B=VectorParameter() {parameterName=Portal_LightColor, defaultValue=<rgba 0.4704 0.225 1 1>}.RGB)
  n21 = VectorParameter() {parameterName=Portal_ShadowColor, defaultValue=<rgba 0.004 0.002 0.02 1>}
  MaterialFunctionCall(Emissive=Multiply(A=ParticleColor().RGB, B=Add(A=@n20, B=@n21.RGB))) {materialFunction=&MF_FX_EmissiveAdjustments}
  Multiply(
   A=ScalarParameter() {defaultValue=25, parameterName=Param_5}
   B=MaterialFunctionCall() {materialFunction=&DitherTemporalAA_Fortnite}.Result
  )
  n22 = Add(A=NamedRerouteUsage()) {constB=-0.5}
  n23 = VectorParameter() {parameterName=Main_UV, defaultValue=<rgba 3 2.101 -0.2 0.248>}
  n24 = VectorParameter() {parameterName=Time_Mul, defaultValue=<rgba 1 1 1 1>}
  NamedRerouteDeclaration(None=MaterialFunctionCall!(
    ClampTexture
    UV=MaterialFunctionCall(Vector or UVs=@n22) {materialFunction=&VectorToRadialValue}.Radial Coordinates
    Tex
    InRGB=@n23.RGB
    InA=@n23.RGB
    Time
    TimeMult=@n24.R
   ) {materialFunction=&/MF_MOON_TexPanner_V3}.PannerUV) {name=MainUV, nodeColor=<rgba 0 0.4049 1 1>}
  NamedRerouteDeclaration(None=@n1) {name=BumpUV, nodeColor=<rgba 0.6406 0 1 1>}
  n25 = MaterialFunctionCall(Float2=NamedRerouteUsage()) {materialFunction=&BreakOutFloat2Components}
  n26 = Multiply(A=@n25.R, B=ScalarParameter() {defaultValue=1, group=UV, parameterName=UV_Tilt})
  NamedRerouteDeclaration(None=MaterialFunctionCall(X=@n25.R, Y=Add(A=@n25.G, B=@n26)) {materialFunction=&MakeFloat2}.Result) {name=TiltedUV, nodeColor=<rgba 0 0.3848 1 1>}
  n27 = VectorParameter() {parameterName=Distortion_UV1, defaultValue=<rgba 0.25 0.5 0.2 0.124>}
  NamedRerouteDeclaration(None=MaterialFunctionCall(UV=NamedRerouteUsage(), InRGB=@n27.RGB, InA=@n27.RGB, TimeMult=@n24.G) {materialFunction=&MF_MOON_TexPanner_V3}.PannerUV) {name=UV_1, nodeColor=<rgba 0 0.4049 1 1>}
  n28 = VectorParameter() {parameterName=Distortion_UV2, defaultValue=<rgba 0.25 0.25 -0.1 0.15>}
  NamedRerouteDeclaration(None=MaterialFunctionCall(UV=NamedRerouteUsage(), InRGB=@n28.RGB, InA=@n28.RGB, TimeMult=@n24.B) {materialFunction=&MF_MOON_TexPanner_V3}.PannerUV) {name=UV_2, nodeColor=<rgba 0 0.4049 1 1>}
  NamedRerouteDeclaration(None=@n18) {name=GradientCenter, nodeColor=<rgba 1 0.6797 0 1>}
  NamedRerouteUsage()
  Power(Base=@n19, Exp=ScalarParameter() {defaultValue=1, parameterName=PortalPower})
 ]
}
