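# Placeholder Niagara effect for a large purple storm portal in the S7 environment.
# Layered glow sprites and slowly rotating smoke meshes form the portal, with curve-paced, distance-culled cloud flashes.
# Lightning meshes spawn probabilistically in two groups, near the center or on a wide ring, growing with camera distance and animated through dynamic material parameters.
# One particle feeds three portal sprite renderers and drives a light modulated by System.Flicker.
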NiagaraSystem /SparksMaps/Effects/Environment/S7/NS_StormPortal_Placeholder
dependencies {
 usedDataInterfaces [&NiagaraDataInterfaceCurve, &NiagaraDataInterfaceArrayFloat3, &NiagaraDataInterfaceArrayFloat]
 usedDynamicInputs []
 usedModules [
  &SystemState
  /Niagara/Noise/Waveform
  &self:SystemSpawnScript.SystemScriptSource.SystemScriptGraph.NiagaraNodeAssignment_0.SetVariables_#1
 ]
}
emitters {
 Center_Light {
  emitterUpdateScript [
   &EmitterState(
    Inactive Response=::NewEnumerator1
    Life Cycle Mode=::NewEnumerator1
    Loop Behavior=::NewEnumerator1
    Loop Delay=1.5
    Loop Duration Mode=::NewEnumerator1
    UseLoopDelay=false
   )
   &SpawnBurst_Instantaneous(Use Spawn Probability=false, UseLoopCountLimit=false)
   &SpawnRate(Spawn Probability=0.5, Use Spawn Probability=false) {enabled=false}
  ]
  particleSpawnScript [
   &InitializeParticle(
    Color=<rgba 22.1718 6.3214 50 1>
    Lifetime=5
    Mass Mode=::NewEnumerator1
    Position Offset=&AddVectorToPosition(Position=$System.PortalPosition, Vector=<0 0 500>)
    Sprite Size Mode=::NewEnumerator3
    Uniform Sprite Size=12500
    UsePositionOffset=true
    Write Material Random=true
   ) {module=InitializeParticle001}
   &SpriteFacingAndAlignment(
    Alignment Coordinate Space=::Local
    Facing Coordinate Space=::Local
    Sprite Alignment=<0 -1 0>
    Sprite Facing=<0 0 -1>
    WriteAlignment=true
    WriteFacing=true
   )
   &ShapeLocation(MaskSpawnGroup=false, Override Seed=false, Sphere Radius=350, UseOwnerScale=true)
  ]
  particleUpdateScript [
   &ParticleState(Kill Particles When Lifetime Has Elapsed=false)
   &ScaleSpriteSize(Scale Factor=&Vector2DFromFloat(Value=&FloatFromCurve(FloatCurve=DataInterface() {propertyValues {Curve {keys [
          <at 0 0>
          {arriveTangent=1.1379, interpMode=RCIM_Cubic, leaveTangent=1.1379, time=0.157, value=0.5261}
          {arriveTangent=0.3249, interpMode=RCIM_Cubic, leaveTangent=0.3249, time=0.7034, value=0.8003}
          <at 1 0.8>
         ]}}}))) {enabled=false}
   &SpriteRotationRate(Rotation Rate=&FloatFromCurve(FloatCurve=DataInterface() {propertyValues {Curve {keys [<at 0 1>,<at 1 0>]}}}, Scale Curve=4))
   &ScaleColor(
    Scale Alpha=&RandomRangeFloat(Minimum=1, Maximum=2, Evaluation Type=::NewEnumerator1)
    ScaleA=true
    ScaleRGB=true
   )
  ]
  rendererClasses [&NiagaraSpriteRendererProperties]
  renderers [{rendererClass=&NiagaraSpriteRendererProperties}]
 }
 Center_Light_Front {
  emitterUpdateScript [
   &EmitterState(
    Inactive Response=::NewEnumerator1
    Life Cycle Mode=::NewEnumerator1
    Loop Behavior=::NewEnumerator1
    Loop Delay=1.5
    Loop Duration Mode=::NewEnumerator1
    UseLoopDelay=false
   )
   &SpawnRate(Spawn Probability=0.5, Use Spawn Probability=false) {enabled=false}
   &SpawnBurst_Instantaneous(Use Spawn Probability=false, UseLoopCountLimit=false)
  ]
  particleSpawnScript [&InitializeParticle(
    Color=&RandomRangeLinearColor(Minimum=<rgba 5 0.0448 20 1>, Maximum=<rgba 9.4615 3.8542 20 1>)
    Lifetime=5
    Mass Mode=::NewEnumerator1
    Position Offset=&AddVectorToPosition(Position=$System.PortalPosition, Vector=<0 0 500>)
    Sprite Size Mode=::NewEnumerator3
    Uniform Sprite Size=12000
    UsePositionOffset=true
    Write Material Random=true
   ) {module=InitializeParticle001}]
  particleUpdateScript [
   &ParticleState(Kill Particles When Lifetime Has Elapsed=false)
   &ScaleSpriteSize(Scale Factor=&Vector2DFromFloat(Value=&FloatFromCurve(FloatCurve=DataInterface() {propertyValues {Curve {keys [
          <at 0 0>
          {arriveTangent=1.1379, interpMode=RCIM_Cubic, leaveTangent=1.1379, time=0.157, value=0.5261}
          {arriveTangent=0.3249, interpMode=RCIM_Cubic, leaveTangent=0.3249, time=0.7034, value=0.8003}
          <at 1 0.8>
         ]}}}))) {enabled=false}
   &SpriteRotationRate(Rotation Rate=&FloatFromCurve(
     FloatCurve=DataInterface() {propertyValues {Curve {keys [<at 0 1>,<at 1 0>]}}}
     Scale Curve=&RandomRangeFloat(Minimum=7, Maximum=-7)
    )) {enabled=false}
   &ScaleColor(
    Scale Alpha=&RandomRangeFloat(Minimum=1, Maximum=1.5, Evaluation Type=::NewEnumerator1)
    Scale RGB=<2 2 2>
    ScaleA=true
    ScaleRGB=true
   )
  ]
  rendererClasses [&NiagaraSpriteRendererProperties]
  renderers [{rendererClass=&NiagaraSpriteRendererProperties}]
 }
 Cloud_LightningFlashes {
  emitterUpdateScript [
   &EmitterState(
    Enable Distance Culling=true
    Life Cycle Mode=::NewEnumerator1
    Loop Delay=1.5
    Loop Duration=8
    MaxDistance=4500
    Module.bUseMaxDistance=false
    Module.bUseMinDistance=false
    Scalability Mode=::NewEnumerator1
    UseLoopDelay=false
   )
   &SpawnRate(
    SpawnRate=&Multiply_Float(
     A=1
     B=&FloatFromCurve(
      FloatCurve=DataInterface() {propertyValues {Curve {keys [
          <at 0 0.25 RCIM_Cubic>
          {arriveTangent=0.0521, interpMode=RCIM_Cubic, leaveTangent=0.0521, time=1.2, value=0.875}
          {arriveTangent=-0.125, interpMode=RCIM_Cubic, leaveTangent=-0.125, time=2.4472, value=0.3776}
          {arriveTangent=-0.11, interpMode=RCIM_Cubic, leaveTangent=-0.11, time=3.6, value=0.575}
          {arriveTangent=0.2109, interpMode=RCIM_Cubic, leaveTangent=0.2109, time=5.034, value=0.0931}
          {arriveTangent=0.1456, interpMode=RCIM_Cubic, leaveTangent=0.1456, time=6.8, value=1.25}
          <at 8 0.525>
         ]}}}
      CurveIndex=$Emitter.NormalizedLoopAge
     )
    )
    Use Spawn Probability=false
   )
  ]
  particleSpawnScript [
   &InitializeParticle(
    Color=&RandomRangeLinearColor(Minimum=<rgba 8.5795 0.3459 20 1>, Maximum=<rgba 11.9122 1.0065 30 1>)
    Lifetime=&RandomRangeFloat(Minimum=0.6, Maximum=0.89)
    Mass Mode=::NewEnumerator1
    Position Offset=&AddVectorToPosition(Position=$System.PortalPosition, Vector=&RandomRangeVector(Minimum=<0 0 1000>, Maximum=<1 1 500>))
    Sprite Rotation Angle=&RandomRangeFloat(Minimum=0, Maximum=1)
    Sprite Rotation Mode=::NewEnumerator0
    Sprite Size Mode=::NewEnumerator1
    Uniform Sprite Size Max=12000
    Uniform Sprite Size Min=10000
    UsePositionOffset=true
    Write Material Random=true
   ) {module=InitializeParticle001}
   &ShapeLocation(
    Large Radius=1500
    MaskSpawnGroup=false
    Non Uniform Scale=<1 3 1>
    Override Seed=false
    Shape Primitive=::NewEnumerator3
    UseOwnerScale=true
   )
   &SpriteFacingAndAlignment(
    Alignment Coordinate Space=::Local
    Facing Coordinate Space=::Local
    Sprite Alignment=<0 -1 0>
    Sprite Facing=<0 0 -1>
    WriteAlignment=true
    WriteFacing=true
   )
  ]
  particleUpdateScript [
   &ParticleState()
   &ScaleSpriteSize(Scale Factor=&Vector2DFromFloat(Value=&FloatFromCurve(FloatCurve=DataInterface() {propertyValues {Curve {keys [
          <at 0 0.6>
          {arriveTangent=0.2848, interpMode=RCIM_Cubic, leaveTangent=0.2848, time=0.2, value=0.7}
          {arriveTangent=0.1875, interpMode=RCIM_Cubic, leaveTangent=0.1875, time=0.7034, value=0.8003}
          <at 1 0.85>
         ]}}})))
   &SpriteRotationRate(Rotation Rate=$Particles.SpriteRotationRate)
   &ScaleColor(
    Scale Alpha=&FloatFromCurve(
     FloatCurve=DataInterface() {propertyValues {Curve {keys [<at 0 1>,<at 1 0>]}}}
     Scale Curve=&RandomRangeFloat(Minimum=10, Maximum=20, Evaluation Type=::NewEnumerator1)
    )
    Scale RGB=&RandomRangeVector(Minimum=<5 1 10>, Maximum=<12 2 30>)
    ScaleA=true
    ScaleRGB=true
   )
  ]
  rendererClasses [&NiagaraSpriteRendererProperties]
  renderers [{rendererClass=&NiagaraSpriteRendererProperties}]
 }
 Lightning {
  emitterSpawnScript [/SetVariables(Emitter.Random:NiagaraFloat=&RandomRangeFloat(Minimum=1, Maximum=50)) {isSetParameters=true}]
  emitterUpdateScript [
   &EmitterState()
   &SpawnRate(Spawn Probability=0.2, SpawnGroup=1, SpawnRate=5, Use Spawn Probability=true)
   &SpawnRate(Spawn Probability=0.2, SpawnRate=1.5, Use Spawn Probability=true) {module=SpawnRate001}
   &SpawnBurst_Instantaneous(Spawn Count=4, Spawn Time=1.5, Use Spawn Probability=false) {enabled=false}
   &SpawnBurst_Instantaneous(Spawn Count=1, Spawn Group=1, Spawn Time=1.5, Use Spawn Probability=false) {enabled=false, module=SpawnBurst_Instantaneous001}
  ]
  particleSpawnScript [
   /SetVariables(Particles.VisibilityTag=&MakeCustomIntFromBool(Bool=$User.Pink, True INT=1, False INT=0)) {enabled=false, isSetParameters=true}
   &InitializeParticle(
    Color=<rgba 0.6785 0 1.2 1>
    Lifetime=&RandomRangeFloat(Minimum=0.35, Maximum=0.8)
    Mesh Scale=&Multiply_Vector(
     A=&SelectVectorFromArray(
      Direct Array Index=&SpawnGroup()
      Vector Selection Array=DataInterface() {propertyValues {FloatData [<5 5 5>, <3 3 3>]}}
     )
     B=&Lerp_Vector(
      A=<1 1 1>
      B=<2 2 2>
      Alpha=&FloatFromCurve(FloatCurve=DataInterface() {propertyValues {Curve {keys [<at 5000 0>,<at 15000 1>]}}}, CurveIndex=$Engine.Owner.LODDistance, Scale Curve=5)
     )
    )
    Mesh Scale Mode=::NewEnumerator4
    UsePositionOffset=false
    Write Material Random=true
   ) {module=InitializeParticle001}
   &InitialMeshOrientation(
    Mesh Axis To Orient=<1 0 1>
    Mesh Orientation Mode=::NewEnumerator2
    Orientation Coordinate Space=::World
    Orientation Vector=<0 0 -1>
    Rotation=&RandomRangeVector(Minimum=<-0.5 -0.05 0>, Maximum=<0.5 -0.1 0>)
    Use Rotation Vector=true
   ) {module=InitialMeshOrientation001}
   &DynamicMaterialParameters(
    Index 0 Param 1=50
    Index 0 Param 2=&UniformRangedFloat(Minimum=1, Maximum=5)
    Index 0 Param 3=10
    Index 0 Param 4=1
    Module.Param0WriteEnabled=true
    Module.Param1WriteEnabled=true
    Module.Param2WriteEnabled=true
    Module.Param3WriteEnabled=true
   )
   &ShapeLocation(MaskSpawnGroup=false, Override Seed=false, Sphere Radius=15000, UseOwnerScale=true) {enabled=false}
   &ShapeLocation(MaskSpawnGroup=false, Override Seed=false, Sphere Radius=600, UseOwnerScale=true) {module=ShapeLocation003}
   &ShapeLocation(
    MaskSpawnGroup=true
    Override Seed=false
    Ring Radius=15000
    Rotation Mode=::NewEnumerator2
    Shape Primitive=::NewEnumerator4
    Spawn Group Mask=1
    U Distribution=0.25
    UseOwnerScale=true
    Yaw / Pitch / Roll=<-0.125 -0.035 0.035>
   ) {module=ShapeLocation002}
  ]
  particleUpdateScript [
   &ParticleState()
   &ScaleColor(
    Scale Alpha=&FloatFromCurve(FloatCurve=DataInterface() {propertyValues {Curve {keys [
         <at 0 0>
         {interpMode=RCIM_Cubic, tangentMode=RCTM_User, time=0.08, value=0.1952}
         {arriveTangent=0.3413, interpMode=RCIM_Cubic, leaveTangent=0.3413, time=0.2, value=0.0902}
         {interpMode=RCIM_Cubic, tangentMode=RCTM_User, time=0.25, value=0.2532}
         <at 1 0>
        ]}}})
    Scale RGB=&VectorFromFloat(Value=&FloatFromCurve(
      FloatCurve=DataInterface() {propertyValues {Curve {keys [
          <at 0 0.5>
          {arriveTangent=14, interpMode=RCIM_Cubic, leaveTangent=14, time=0.24, value=1}
          {arriveTangent=-4.524, interpMode=RCIM_Cubic, leaveTangent=-4.524, tangentMode=RCTM_User, time=0.25, value=4}
          <at 1 1>
         ]}}}
      Scale Curve=&RandomRangeFloat(Minimum=1, Maximum=100)
     ))
    ScaleA=true
    ScaleRGB=true
   )
   &DynamicMaterialParameters(
    Index 0 Param 1=&FloatFromCurve(FloatCurve=DataInterface() {propertyValues {Curve {keys [
         {arriveTangent=-2.3592, time=0, value=1}
         {arriveTangent=-1.6922, leaveTangent=-1.6922, time=0.25, value=0.5}
         {arriveTangent=-0.9805, interpMode=RCIM_Cubic, leaveTangent=-0.9805, tangentMode=RCTM_User, time=1, value=0}
        ]}}})
    Index 0 Param 2=&FloatFromCurve(
     FloatCurve=DataInterface() {propertyValues {Curve {keys [
         {interpMode=RCIM_Cubic, leaveTangent=-1.1405, tangentMode=RCTM_Break, time=0, value=0.5}
         {arriveTangent=4.4936, interpMode=RCIM_Cubic, leaveTangent=-3.0641, tangentMode=RCTM_Break, time=0.25, value=1}
         <at 1 0>
        ]}}}
     Scale Curve=&RandomRangeFloat(Minimum=2, Maximum=5)
    )
    Index 0 Param 3=&FloatFromCurve(
     FloatCurve=DataInterface() {propertyValues {Curve {keys [
         <at 0 1 RCIM_Cubic>
         {arriveTangent=-3.8306, interpMode=RCIM_Cubic, tangentMode=RCTM_Break, time=0.24, value=0.1}
         {arriveTangent=-2.288, interpMode=RCIM_Cubic, leaveTangent=-2.288, tangentMode=RCTM_User, time=0.25, value=1}
         <at 1 0.2 RCIM_Cubic>
        ]}}}
     Scale Curve=10
    )
    Index 3 Param 1=3
    Index 3 Param 2=1000
    Module.Index0Param0Write=true
    Module.Index0Param1Write=true
    Module.Index0Param2Write=true
    Module.Index0Param3Write=false
    Module.Index3Param0Write=true
    Module.Index3Param1Write=true
    Module.Index3Param2Write=false
    Module.Index3Param3Write=false
    Write Parameter Index 3=true
   ) {module=DynamicMaterialParameters001}
  ]
  rendererClasses [&NiagaraMeshRendererProperties]
  renderers [
   {rendererClass=&NiagaraMeshRendererProperties, rendererIndex=0}
   {rendererClass=&NiagaraMeshRendererProperties, rendererIndex=1}
  ]
 }
 Portal_Combo {
  emitterUpdateScript [
   &EmitterState()
   &SpawnBurst_Instantaneous(Use Spawn Probability=false, UseLoopCountLimit=true)
  ]
  particleSpawnScript [
   &InitializeParticle(Color Mode=::NewEnumerator0, Lifetime=5, Position Offset=<0 0 779.3087>, UsePositionOffset=false, Write Material Random=true)
   &SpriteFacingAndAlignment(Alignment Coordinate Space=::Local, Facing Coordinate Space=::Local, Sprite Facing=<0 0 1>, WriteAlignment=true, WriteFacing=true)
   /SetVariables(Particles.Position00:NiagaraPosition=&AddVectorToPosition(Position=$Engine.Emitter.SimulationPosition, Vector=<0 0 780>)) {isSetParameters=true}
   /SetVariables(Particles.Position01:NiagaraPosition=&AddVectorToPosition(Position=$Engine.Emitter.SimulationPosition, Vector=<0 0 800>)) {isSetParameters=true}
   /SetVariables(Particles.Position02:NiagaraPosition=&AddVectorToPosition(Position=$Engine.Emitter.SimulationPosition, Vector=<0 0 100>)) {isSetParameters=true}
   /SetVariables(Particles.SpriteSize00:Vector2f=<42000 42000>) {isSetParameters=true}
   /SetVariables(Particles.SpriteSize01:Vector2f=<38000 38000>) {isSetParameters=true}
   /SetVariables(Particles.SpriteSize02:Vector2f=<30000 30000>) {isSetParameters=true}
  ]
  particleUpdateScript [
   &ParticleState(Kill Particles When Lifetime Has Elapsed=false)
   &SpriteRotationRate(Rotation Rate=0) {enabled=false}
   &FN_Mesh_Attributes(Create Color Attribute=true, Create Scale Attribute=true, Scale=<12 12 12>, Write to Module Namespace=true)
   &FN_Gen5_Light(
    Attenutation=1156.5077
    Color:LinearColor=<rgba 0.7033 0.11 1 1>
    Exponent=3
    Intensity:NiagaraFloat=&Multiply_Float(A=62.3144, B=&Lerp_Float(A=0.2, B=1, Alpha=$System.Flicker))
   )
  ]
  rendererClasses [&NiagaraSpriteRendererProperties]
  renderers [
   {rendererClass=&NiagaraSpriteRendererProperties}
   {rendererClass=&NiagaraSpriteRendererProperties, rendererIndex=1}
   {rendererClass=&NiagaraSpriteRendererProperties, rendererIndex=2}
  ]
 }
 Storm_SmokeCenter {
  emitterUpdateScript [
   &EmitterState(
    Inactive Response=::NewEnumerator1
    Life Cycle Mode=::NewEnumerator1
    Loop Behavior=::NewEnumerator1
    Loop Duration Mode=::NewEnumerator1
    UseLoopDelay=false
   )
   &SpawnBurst_Instantaneous(Use Spawn Probability=false, UseLoopCountLimit=false)
  ]
  particleSpawnScript [
   &InitializeParticle(
    Color Mode=::NewEnumerator0
    Lifetime=5
    Mesh Scale=<35 35 5>
    Mesh Scale Mode=::NewEnumerator4
    Sprite Rotation Mode=::NewEnumerator2
    UsePositionOffset=true
    Write Material Random=true
   )
   &InitialMeshOrientation(
    Mesh Orientation Mode=::NewEnumerator7
    Rotation=&RandomRangeVector(Minimum=<0 0 -1>, Maximum=<0 0 1>)
    Use Rotation Vector=true
   )
  ]
  particleUpdateScript [
   &DynamicMaterialParameters(
    Index 1 Param 1=2
    Index 1 Param 2=0.7
    Index 1 Param 3=1
    Module.Index1Param0Write=true
    Module.Index1Param1Write=true
    Module.Index1Param2Write=true
    Module.Index1Param3Write=true
    Write Parameter Index 0=false
    Write Parameter Index 1=true
   )
   &ParticleState(Kill Particles When Lifetime Has Elapsed=false)
   &UpdateMeshOrientation(Rotation Rate=-0.008, Rotation Vector=<0 0 1>)
   &FN_Mesh_Attributes(Create Color Attribute=true, Create Scale Attribute=true, Scale=<12 12 12>, Write to Module Namespace=true) {enabled=false}
   &FN_Gen5_Light(
    Attenutation=1156.5077
    Color:LinearColor=<rgba 0.7033 0.11 1 1>
    Exponent=3
    Intensity:NiagaraFloat=&Multiply_Float(A=62.3144, B=&Lerp_Float(A=0.2, B=1, Alpha=$System.Flicker))
   ) {enabled=false}
   &ScaleColor(Scale RGB=<2 1 3>, ScaleA=true, ScaleRGB=true)
  ]
  rendererClasses [&NiagaraMeshRendererProperties]
  renderers [{rendererClass=&NiagaraMeshRendererProperties, rendererIndex=0}]
 }
 Storm_SmokeMiddle {
  emitterUpdateScript [
   &EmitterState(
    Inactive Response=::NewEnumerator1
    Life Cycle Mode=::NewEnumerator1
    Loop Behavior=::NewEnumerator1
    Loop Duration Mode=::NewEnumerator1
    UseLoopDelay=false
   )
   &SpawnBurst_Instantaneous(Spawn Count=2, Use Spawn Probability=false, UseLoopCountLimit=false)
  ]
  particleSpawnScript [
   &PartitionParticles()
   &InitializeParticle(
    Color Mode=::NewEnumerator0
    Lifetime=5
    Mesh Scale=&SelectVectorFromArray(
     Direct Array Index=$Particles.PartitionParticles.Partition
     Vector Selection Array=DataInterface() {propertyValues {FloatData [<130 130 50>, <180 180 80>]}}
    )
    Mesh Scale Mode=::NewEnumerator4
    Position Offset=&SelectVectorFromArray(
     Direct Array Index=$Particles.PartitionParticles.Partition
     Vector Selection Array=DataInterface() {propertyValues {FloatData [<0 0 575>, <0 0 1245>]}}
    )
    Sprite Rotation Mode=::NewEnumerator2
    UsePositionOffset=true
    Write Material Random=true
   )
   &InitialMeshOrientation(Mesh Orientation Mode=::NewEnumerator1, Rotation=<0.5 0 0>, Use Rotation Vector=true)
  ]
  particleUpdateScript [
   &DynamicMaterialParameters(
    Index 1 Param 1=3
    Index 1 Param 2=0.72
    Index 1 Param 3=1
    Module.Index1Param0Write=true
    Module.Index1Param1Write=true
    Module.Index1Param2Write=true
    Module.Index1Param3Write=true
    Write Parameter Index 0=false
    Write Parameter Index 1=true
   )
   &ParticleState(Kill Particles When Lifetime Has Elapsed=false)
   &UpdateMeshOrientation(
    Rotation Rate=&SelectFloatFromArray(
     Direct Array Index=$Particles.Initial.PartitionParticles.Partition
     Float Selection Array=DataInterface() {propertyValues {FloatData [0.01,0.002]}}
    )
    Rotation Vector=<0 0 1>
   )
   &FN_Mesh_Attributes(Create Color Attribute=true, Create Scale Attribute=true, Scale=<12 12 12>, Write to Module Namespace=true) {enabled=false}
   &FN_Gen5_Light(
    Attenutation=1156.5077
    Color:LinearColor=<rgba 0.7033 0.11 1 1>
    Exponent=3
    Intensity:NiagaraFloat=&Multiply_Float(A=62.3144, B=&Lerp_Float(A=0.2, B=1, Alpha=$System.Flicker))
   ) {enabled=false}
   &ScaleColor(
    Scale RGB=&SelectVectorFromArray(
     Direct Array Index=$Particles.Initial.PartitionParticles.Partition
     Vector Selection Array=DataInterface() {propertyValues {FloatData [<1.5 1 2>, <0.3 0.4 0.5>]}}
    )
    ScaleA=true
    ScaleRGB=true
   )
  ]
  rendererClasses [&NiagaraMeshRendererProperties]
  renderers [{rendererClass=&NiagaraMeshRendererProperties, rendererIndex=0}]
 }
}
summary {
 systemName=NS_StormPortal_Placeholder
 emitters [
  {emitterName=Center_Light, rendererClasses [&NiagaraSpriteRendererProperties]}
  {emitterName=Cloud_LightningFlashes, rendererClasses [&NiagaraSpriteRendererProperties]}
  {emitterName=Storm_SmokeCenter, rendererClasses [&NiagaraMeshRendererProperties]}
  {emitterName=Lightning, rendererClasses [&NiagaraMeshRendererProperties]}
  {emitterName=Center_Light_Front, rendererClasses [&NiagaraSpriteRendererProperties]}
  {emitterName=Portal_Combo, rendererClasses [&NiagaraSpriteRendererProperties]}
  {emitterName=Storm_SmokeMiddle, rendererClasses [&NiagaraMeshRendererProperties]}
 ]
 userVariables [
  User.bIsFrontEnd {type=NiagaraBool}
  User.bIsFrontEndPreview {type=NiagaraBool}
  User.InitialStartDelay {type=NiagaraFloat, defaultValue=0, description=""}
  User.Open {type=NiagaraFloat, defaultValue=1, description=""}
  User.Pink {type=NiagaraBool}
  User.RampupTimeSmoke {type=NiagaraFloat, defaultValue=1, description=""}
 ]
}
